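# Physics asset for the pendulum on the Octoberfest backpack accessory, letting the piece swing as a secondary-motion element.
# It has two capsule bodies: a large one on backpack_root and a slimmer, elongated one on dyn_pendulum_01.
# A single constraint joins the pendulum to the root with twist and swing 1 locked, so it rotates only about the swing 2 axis, like a hinge.
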PhysicsAsset /Game/Accessories/FORT_Backpacks/M_MED_Octoberfest_Backpack/Meshes/M_MED_Octoberfest_Backpack_Pendulum_Physics
bodies {
 backpack_root [{shapeType=Capsule, length=15, radius=21.25, center=<-0.6003 3.1821 0.0009>}]
 dyn_pendulum_01 [{shapeType=Capsule, length=22.2398, radius=5.3718, center=<16.7296 -4.0283 0.0151>, rotation=<rot 0 91.2223 -89.9958>}]
}
constraints [{
  bone1Name=dyn_pendulum_01
  bone2Name=backpack_root
  swing1LimitDegrees=45
  swing1Motion=Locked
  swing2LimitDegrees=60
  twistLimitDegrees=45
  twistMotion=Locked
 }]
