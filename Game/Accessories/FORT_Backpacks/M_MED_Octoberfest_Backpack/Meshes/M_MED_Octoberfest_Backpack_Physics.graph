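# Physics asset for the Octoberfest backpack accessory, defining capsule bodies and joints for its dangling and moving decorations.
# Front and rear pendulum cones, a mobile and four trees form two-link chains anchored to backpack_root, with limits on swing and twist.
# The bell and wheel lock both swing axes, so they only rotate about twist, the wheel almost half a turn each way.
# The root carries two capsules, and backpack_bottom joins it with all angular axes free.
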PhysicsAsset /Game/Accessories/FORT_Backpacks/M_MED_Octoberfest_Backpack/Meshes/M_MED_Octoberfest_Backpack_Physics
bodies {
 backpack_bottom [{shapeType=Capsule, length=19.5, radius=14.75, center=<-1.0048 0 -0.0001>, rotation=<rot 0 -90 0>}]
 backpack_root [
  {shapeType=Capsule, length=85, radius=20, center=<-39.9396 14.3163 0>, rotation=<rot 90 -176.4237 -20.5327>}
  {shapeType=Capsule, length=15, radius=21.25, center=<-0.6003 3.1821 0.0009>}
 ]
 dyn_bell [{shapeType=Capsule, length=6.25, radius=3.5, center=<-0.0223 0.0322 -6.506>, rotation=<rot 1.6007 -89.9995 -0.7855>}]
 dyn_mobile_01 [{shapeType=Capsule, length=0, radius=2.75, center=<1.3026 0.0128 -0.0075>, rotation=<rot 0 90 -90>}]
 dyn_mobile_02 [{shapeType=Capsule, length=4.1063, radius=2.2872, center=<4.05 -0.1687 0.1033>, rotation=<rot 0 9.3748 -60.4639>}]
 dyn_pend_cone_front_01 [{shapeType=Capsule, length=3.75, radius=2, center=<5.2708 -0.0012 0.0011>, rotation=<rot 0 90 -90>}]
 dyn_pend_cone_front_02 [{shapeType=Capsule, length=4.6358, radius=2, center=<7.8322 0.0105 -0.0292>, rotation=<rot 0 -89.9229 -89.8427>}]
 dyn_pend_cone_rear_01 [{shapeType=Capsule, length=5.75, radius=2.75, center=<0.3919 -0.0138 0.0081>, rotation=<rot 0 90 -90>}]
 dyn_pend_cone_rear_02 [{shapeType=Capsule, length=5.6879, radius=1.6975, center=<6.5101 -0.0013 -0.0183>, rotation=<rot 0 -89.89 -89.9999>}]
 dyn_treea_01 [{shapeType=Capsule, length=1.1365, radius=2.5624, center=<0 0 -0.4967>}]
 dyn_treea_02 [{shapeType=Capsule, length=0.6365, radius=2.0624}]
 dyn_treeb_01 [{shapeType=Capsule, length=1.1365, radius=2.5624}]
 dyn_treeb_02 [{shapeType=Capsule, length=0.6365, radius=2.0624}]
 dyn_treec_01 [{shapeType=Capsule, length=1.1365, radius=2.5624, center=<0 0 -0.5039>}]
 dyn_treec_02 [{shapeType=Capsule, length=0.6365, radius=2.0624, center=<0.1616 0 0.4737>}]
 dyn_treed_01 [{shapeType=Capsule, length=1.1365, radius=2.5624}]
 dyn_treed_02 [{shapeType=Capsule, length=0.6365, radius=2.0624, center=<0 0 -0.2088>}]
 dyn_wheel [{shapeType=Capsule, length=0.75, radius=8.01}]
}
constraints [
 {bone1Name=dyn_pend_cone_front_01, bone2Name=backpack_root, swing1LimitDegrees=72.5714, swing2LimitDegrees=40, twistLimitDegrees=15}
 {bone1Name=dyn_pend_cone_front_02, bone2Name=dyn_pend_cone_front_01, swing1LimitDegrees=35, swing2LimitDegrees=35, twistLimitDegrees=8}
 {bone1Name=dyn_pend_cone_rear_01, bone2Name=backpack_root, swing1LimitDegrees=81.4286, swing2LimitDegrees=30, twistLimitDegrees=8}
 {bone1Name=dyn_pend_cone_rear_02, bone2Name=dyn_pend_cone_rear_01, swing1LimitDegrees=35, swing2LimitDegrees=35, twistLimitDegrees=7.2857}
 {bone1Name=dyn_mobile_01, bone2Name=backpack_root, swing1LimitDegrees=5.5714, swing2LimitDegrees=27.4286, twistLimitDegrees=7.2857}
 {bone1Name=dyn_mobile_02, bone2Name=dyn_mobile_01, swing1LimitDegrees=30, swing2LimitDegrees=25, twistLimitDegrees=13}
 {
  bone1Name=dyn_bell
  bone2Name=backpack_root
  swing1LimitDegrees=45
  swing1Motion=Locked
  swing2LimitDegrees=45
  swing2Motion=Locked
  twistLimitDegrees=60
 }
 {
  bone1Name=dyn_wheel
  bone2Name=backpack_root
  swing1LimitDegrees=45
  swing1Motion=Locked
  swing2LimitDegrees=45
  swing2Motion=Locked
  twistLimitDegrees=178
 }
 {bone1Name=dyn_treea_01, bone2Name=backpack_root, swing1LimitDegrees=45, swing2LimitDegrees=45, twistLimitDegrees=45}
 {bone1Name=dyn_treea_02, bone2Name=dyn_treea_01, swing1LimitDegrees=45, swing2LimitDegrees=45, twistLimitDegrees=45}
 {bone1Name=dyn_treeb_01, bone2Name=backpack_root, swing1LimitDegrees=45, swing2LimitDegrees=45, twistLimitDegrees=45}
 {bone1Name=dyn_treeb_02, bone2Name=dyn_treeb_01, swing1LimitDegrees=45, swing2LimitDegrees=45, twistLimitDegrees=45}
 {bone1Name=dyn_treec_01, bone2Name=backpack_root, swing1LimitDegrees=45, swing2LimitDegrees=45, twistLimitDegrees=45}
 {bone1Name=dyn_treec_02, bone2Name=dyn_treec_01, swing1LimitDegrees=45, swing2LimitDegrees=45, twistLimitDegrees=45}
 {bone1Name=dyn_treed_01, bone2Name=backpack_root, swing1LimitDegrees=45, swing2LimitDegrees=45, twistLimitDegrees=45}
 {bone1Name=dyn_treed_02, bone2Name=dyn_treed_01, swing1LimitDegrees=45, swing2LimitDegrees=45, twistLimitDegrees=45}
 {
  bone1Name=backpack_bottom
  bone2Name=backpack_root
  swing1LimitDegrees=45
  swing1Motion=Free
  swing2LimitDegrees=45
  swing2Motion=Free
  twistLimitDegrees=45
  twistMotion=Free
 }
]
