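# Animation Blueprint for the Oktoberfest cuckoo-clock backpack, animating its llama cuckoo, doors, pendulum and clock hands.
# bIsCuckooOut drives nested blends of in/out poses and the input pose, feeding AnimDynamics on llama and door bones.
# Variable-driven bone transforms and two RigidBody sims follow, ending with clock hands and llama scale.
# Each update, the event graph triggers the part modifier's FX and sound once per cuckoo-out in the front end.
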AnimBlueprint /BRCosmetics/Accessories/FORT_Backpacks/M_MED_Octoberfest_Backpack/Meshes/M_MED_Octoberfest_Backpack_Skeleton_AnimBlueprint
target_skeleton=/Game/Accessories/FORT_Backpacks/M_MED_Octoberfest_Backpack/Meshes/M_MED_Octoberfest_Backpack_Skeleton
graphs [
 AnimGraph {nodes [
   n1 = Get CuckooDynamicsAlpha() -> (CuckooDynamicsAlpha:float)
   Anim Dynamics - Bone: dyn_llama_top(Alpha=@n1, ExternalForce="0,0,0") {
    bGravityOverrideInSimSpace=true
    alphaBoolBlend {blendOption=Linear}
    alphaScaleBiasClamp {interpSpeedDecreasing=0.15, interpSpeedIncreasing=0.25}
    boundBone {boneName=dyn_llama_top}
    physicsBodyDefinitions [{
      sphereCollisionRadius=0
      boundBone {boneName=dyn_llama_top}
      boxExtents=<5 5 1>
      constraintSetup {twistAxis=AxisZ, angularLimitsMax=<30 20 0>, angularLimitsMin=<-30 -20 0>, angularTarget=<0 0 0>}
      localJointOffset=<-0.1 -0.1 -1>
     }]
   }
   n2 = M_MED_Octoberfest_Backpack_CuckooD_PoseAsset : cuckoo_in!(PoseAsset:Pose Asset Object Reference=/Game/Accessories/FORT_Backpacks/M_MED_Octoberfest_Backpack/Meshes/M_MED_Octoberfest_Backpack_CuckooD_PoseAsset) -> (Pose:Pose Link Structure) {is=title AnimGraphNode_PoseByName, poseName=cuckoo_in}
   AnimGraphNode_SubInput_0 = Input Pose()
   n3 = Get bIsCuckooOut() -> (bIsCuckooOut:bool)
   n4 = Two Way Blend!(
    A=@AnimGraphNode_SubInput_0
    B=M_MED_Octoberfest_Backpack_CuckooD_PoseAsset : cuckoo_out!(PoseAsset:Pose Asset Object Reference=&M_MED_Octoberfest_Backpack_CuckooD_PoseAsset) -> (Pose:Pose Link Structure) {is=title AnimGraphNode_PoseByName, poseName=cuckoo_out}
    bAlphaBoolEnabled:bool=@n3
   ) -> (Pose) {blendNode {alphaInputType=Bool, alphaBoolBlend {blendInTime=0.5, blendOption=ExpIn, blendOutTime=1}}}
   n5 = Anim Dynamics - Bone: dyn_llama_root(
    ComponentPose=Local To Component(LocalPose=Two Way Blend(A=@n2, B=@n4, bAlphaBoolEnabled=@n3) {blendNode {alphaInputType=Bool, alphaBoolBlend {blendInTime=0.15, blendOption=CircularIn, blendOutTime=0.15}}})
    Alpha=@n1
    ExternalForce="0,0,0"
   ) {
    bGravityOverrideInSimSpace=true
    alphaBoolBlend {blendOption=Linear}
    alphaScaleBiasClamp {interpSpeedDecreasing=0.15, interpSpeedIncreasing=0.25}
    boundBone {boneName=dyn_llama_root}
    physicsBodyDefinitions [{
      sphereCollisionRadius=0
      boundBone {boneName=dyn_llama_root}
      boxExtents=<15 15 1>
      constraintSetup {angularTargetAxis=AxisZ, twistAxis=AxisZ, angularLimitsMax=<7 20 0>, angularLimitsMin=<-7 -20 0>, angularTarget=<0 0 0>}
      localJointOffset=<-0.1 -0.3 -5>
     }]
   }
   n6 = Anim Dynamics - Bone: dyn_llama_offset(ComponentPose=@n5, Alpha=@n1, ExternalForce="0,0,0") {
    bGravityOverrideInSimSpace=true
    alphaBoolBlend {blendOption=Linear}
    alphaScaleBiasClamp {interpSpeedDecreasing=0.15, interpSpeedIncreasing=0.25}
    boundBone {boneName=dyn_llama_offset}
    physicsBodyDefinitions [{
      sphereCollisionRadius=0
      boundBone {boneName=dyn_llama_offset}
      boxExtents=<10 10 1>
      constraintSetup {angularTargetAxis=AxisZ, twistAxis=AxisZ, angularLimitsMax=<15 20 0>, angularLimitsMin=<-15 -20 0>, angularTarget=<0 0 0>}
      localJointOffset=<-0.1 -0.3 -5>
     }]
   }
   n7 = Anim Dynamics - Bone: dyn_llama_base(ComponentPose=@n6, Alpha=@n1, ExternalForce="0,0,0") {
    bGravityOverrideInSimSpace=true
    alphaBoolBlend {blendOption=Linear}
    alphaScaleBiasClamp {interpSpeedDecreasing=0.15, interpSpeedIncreasing=0.25}
    boundBone {boneName=dyn_llama_base}
    physicsBodyDefinitions [{
      sphereCollisionRadius=0
      boundBone {boneName=dyn_llama_base}
      boxExtents=<9 9 1>
      constraintSetup {angularTargetAxis=AxisZ, twistAxis=AxisZ, angularLimitsMax=<30 20 0>, angularLimitsMin=<-30 -20 0>, angularTarget=<0 0 0>}
      localJointOffset=<-0.1 -0.05 -2>
     }]
   }
   n8 = Anim Dynamics - Bone: dyn_door_l(ComponentPose=@n7, Alpha=@n1, ExternalForce="0,0,0") {
    bGravityOverrideInSimSpace=true
    alphaBoolBlend {blendOption=Linear}
    alphaScaleBiasClamp {interpSpeedDecreasing=0, interpSpeedIncreasing=0.25}
    boundBone {boneName=dyn_door_l}
    physicsBodyDefinitions [{
      sphereCollisionRadius=0
      boundBone {boneName=dyn_door_l}
      boxExtents=<1 10 10>
      constraintSetup {twistAxis=AxisY, angularLimitsMax=<0 0 30>, angularLimitsMin=<0 0 -30>, angularTarget=<0 0 0>}
      localJointOffset=<0 1 0>
     }]
   }
   n9 = Anim Dynamics - Bone: dyn_door_r(ComponentPose=@n8, Alpha=@n1, ExternalForce="0,0,0") {
    bGravityOverrideInSimSpace=true
    alphaBoolBlend {blendOption=Linear}
    alphaScaleBiasClamp {interpSpeedDecreasing=0, interpSpeedIncreasing=0.25}
    boundBone {boneName=dyn_door_r}
    physicsBodyDefinitions [{
      sphereCollisionRadius=0
      boundBone {boneName=dyn_door_r}
      boxExtents=<1 10 10>
      constraintSetup {twistAxis=AxisY, angularLimitsMax=<0 0 30>, angularLimitsMin=<0 0 -30>, angularTarget=<0 0 0>}
      localJointOffset=<0 1 0>
     }]
   }
   n10 = "Transform (Modify) Bone - Bone: backpack_root"(
    ComponentPose="Transform (Modify) Bone - Bone: backpack_bottom"(ComponentPose=@n9, Alpha=1, Translation=Get BoneAPosition() -> (BoneAPosition:Vector), Rotation="0,0,0", Scale="1,1,1") {translationMode=BMM_Additive, alphaBoolBlend {blendOption=Linear}, boneToModify {boneName=backpack_bottom}}
    Alpha=1
    Translation="0,0,0"
    Rotation=Get BackpackRotator() -> (BackpackRotator:Rotator)
    Scale="1,1,1"
   ) {rotationMode=BMM_Additive, rotationSpace=BCS_BoneSpace, alphaBoolBlend {blendOption=Linear}, boneToModify {boneName=backpack_root}}
   n11 = "Transform (Modify) Bone - Bone: dyn_pendulum_base"(ComponentPose=@n10, Alpha=1, Translation="0,0,0", Rotation=Get PendulumRotator() -> (PendulumRotator:Rotator), Scale="1,1,1") {rotationMode=BMM_Replace, rotationSpace=BCS_BoneSpace, alphaBoolBlend {blendOption=Linear}, boneToModify {boneName=dyn_pendulum_base}}
   n12 = RigidBody(ComponentPose=@n11) {
    overridePhysicsAsset=/Game/Accessories/FORT_Backpacks/M_MED_Octoberfest_Backpack/Meshes/M_MED_Octoberfest_Backpack_Physics
    alphaBoolBlend {blendOption=Linear}
    componentAppliedLinearAccClamp=<500 500 500>
    componentLinearAccScale=<1 1 1>
    componentLinearVelScale=<8 8 8>
   }
   n13 = RigidBody(ComponentPose=@n12, Alpha=Get PendulumDynamicsAlpha() -> (PendulumDynamicsAlpha:float)) {
    overridePhysicsAsset=/Game/Accessories/FORT_Backpacks/M_MED_Octoberfest_Backpack/Meshes/M_MED_Octoberfest_Backpack_Pendulum_Physics
    alphaBoolBlend {blendOption=Linear}
    componentAppliedLinearAccClamp=<800 800 800>
    componentLinearAccScale=<3 3 3>
    componentLinearVelScale=<9 9 9>
   }
   n14 = "Transform (Modify) Bone - Bone: dyn_clock_hour"(ComponentPose=@n13, Alpha=1, Translation="0,0,0", Rotation=Get HourHandRotator() -> (HourHandRotator:Rotator), Scale="1,1,1") {rotationMode=BMM_Additive, rotationSpace=BCS_BoneSpace, alphaBoolBlend {blendOption=Linear}, boneToModify {boneName=dyn_clock_hour}}
   n15 = "Transform (Modify) Bone - Bone: dyn_clock_min"(ComponentPose=@n14, Alpha=1, Translation="0,0,0", Rotation=Get MinuteHandRotator() -> (MinuteHandRotator:Rotator), Scale="1,1,1") {rotationMode=BMM_Replace, rotationSpace=BCS_BoneSpace, alphaBoolBlend {blendOption=Linear}, boneToModify {boneName=dyn_clock_min}}
   n16 = "Transform (Modify) Bone - Bone: dyn_llama_base"(ComponentPose=@n15, Alpha=1, Translation="0,0,0", Rotation="0,0,0", Scale=Get CurrentLlamaScale() -> (CurrentLlamaScale:Vector)) {
    scaleMode=BMM_Replace
    scaleSpace=BCS_BoneSpace
    alphaBoolBlend {blendOption=Linear}
    alphaScaleBiasClamp {bInterpResult=true, interpSpeedDecreasing=25, interpSpeedIncreasing=20}
    boneToModify {boneName=dyn_llama_base}
   }
   Output Pose(Result=Component To Local(ComponentPose=@n16)) {layerGroup=None}
  ]}
 EventGraph {nodes [
   Event BlueprintUpdateAnimation()
   n1 = GetOwningActor()
   > Is Valid(InputObject=@n1)
   > n2 = Cast To FortPlayerPawn(Object=@n1)
   > n3 = Cast To B_Athena_PartModifier_OktoberfestMale!(
    execute
    Object=GetModifierForPartType!(self:Fort Player Pawn Object Reference=@n2, PartType:EFortCustomPartType Enum=Backpack) -> (ReturnValue:Custom Character Part Modifier Object Reference)
   ) -> (then, CastFailed, AsB Athena Part Modifier Oktoberfest Male:B Athena Part Modifier Oktoberfest Male Object Reference) {is=title K2Node_DynamicCast}
   > n4 = Branch(Condition=AND Boolean(A=Get bIsCuckooOut(), B=Get bIsInFrontEnd()))
   > Do Once(Reset=@n4.else)
   > PlayFXandSound(self:B Athena Part Modifier Oktoberfest Male Object Reference=@n3)
   Event BlueprintInitializeAnimation()
  ]}
]
variables [As B Athena Part Modifier Oktoberfest Male {type=B_Athena_PartModifier_OktoberfestMale_C, default=null}]
